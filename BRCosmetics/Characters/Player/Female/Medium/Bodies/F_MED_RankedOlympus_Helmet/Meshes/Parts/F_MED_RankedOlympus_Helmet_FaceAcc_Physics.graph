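# Physics asset for the face accessory part of the female medium RankedOlympus helmet, covering three dynamic hair bone chains: left, middle and right.
# Each chain runs from a MSTR bone through four capsule segments, with tight constraints that mostly allow swing on one axis.
# Angularly free constraints tie the third left and right segments to the middle chain.
# Capsules on upper spine, neck, clavicles and upper arms provide body collision shapes.
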PhysicsAsset /BRCosmetics/Characters/Player/Female/Medium/Bodies/F_MED_RankedOlympus_Helmet/Meshes/Parts/F_MED_RankedOlympus_Helmet_FaceAcc_Physics
bodies {
 clavicle_l [{shapeType=Capsule, length=13, radius=7.5, shapeName=clavicle_l_capsule, center=<-1 1.8 0>, rotation=<rot 90 0 0>}]
 clavicle_r [{shapeType=Capsule, length=13, radius=7.5, shapeName=clavicle_r_capsule, center=<-1 -1.8 0>, rotation=<rot -90 0 180>}]
 dyn_hair_left_01 [{shapeType=Capsule, length=2.5, radius=5.5, shapeName=dyn_hair_left_01_capsule, center=<6.5 0 4>, rotation=<rot 0 0 90>}]
 dyn_hair_left_02 [{shapeType=Capsule, length=2, radius=5, shapeName=dyn_hair_left_02_capsule, center=<3 0 3>, rotation=<rot 0 0 83.1371>}]
 dyn_hair_left_03 [{shapeType=Capsule, length=1.5, radius=5.5, shapeName=dyn_hair_left_03_capsule, center=<3 0 2.5>, rotation=<rot 0 0 81.1837>}]
 dyn_hair_left_04 [{shapeType=Capsule, length=2, radius=6.5, shapeName=dyn_hair_left_04_capsule, center=<6.3 0 4.7>, rotation=<rot 0 0 77.6369>}]
 dyn_hair_left_MSTR [{shapeType=Capsule, length=1, radius=0.5, shapeName=dyn_hair_left_MSTR_capsule}]
 dyn_hair_middle_01 [{shapeType=Capsule, length=3, radius=5.5, shapeName=dyn_hair_middle_01_capsule, center=<6.3 0 4>, rotation=<rot 0 0 90>}]
 dyn_hair_middle_02 [{shapeType=Capsule, length=3, radius=6.5, shapeName=dyn_hair_middle_02_capsule, center=<2.7 0 5>, rotation=<rot 0 0 90>}]
 dyn_hair_middle_03 [{shapeType=Capsule, length=3.5, radius=9, shapeName=dyn_hair_middle_03_capsule, center=<4 0 7.5>, rotation=<rot 0 0 90>}]
 dyn_hair_middle_04 [{shapeType=Capsule, length=7, radius=7.5, shapeName=dyn_hair_middle_04_capsule, center=<7 0 6.6>, rotation=<rot 0 0 90>}]
 dyn_hair_middle_MSTR [{shapeType=Capsule, length=1, radius=0.5, shapeName=dyn_hair_middle_MSTR_capsule}]
 dyn_hair_right_01 [{shapeType=Capsule, length=2.5, radius=5.5, shapeName=dyn_hair_right_01_capsule, center=<6.5 0 3.2>, rotation=<rot 0 0 90>}]
 dyn_hair_right_02 [{shapeType=Capsule, length=2, radius=5, shapeName=dyn_hair_right_02_capsule, center=<3.7742 0 2.7>, rotation=<rot 0 0 100.0115>}]
 dyn_hair_right_03 [{shapeType=Capsule, length=1.5, radius=5, shapeName=dyn_hair_right_03_capsule, center=<3.8788 0 3.4746>, rotation=<rot 0 0 90>}]
 dyn_hair_right_04 [{shapeType=Capsule, length=2, radius=6, shapeName=dyn_hair_right_04_capsule, center=<6.3 0 4>, rotation=<rot 0 0 90>}]
 dyn_hair_right_MSTR [{shapeType=Capsule, length=1, radius=0.5, shapeName=dyn_hair_right_MSTR_capsule}]
 neck_01 [{shapeType=Capsule, length=5.5, radius=4, shapeName=neck_01_capsule, center=<3.8795 0 0>, rotation=<rot 90 0 0>}]
 spine_04 [{shapeType=Capsule, length=17, radius=13, shapeName=spine_04_capsule, center=<4.2166 3.549 0>}]
 upperarm_twist_01_l [{
   shapeType=Capsule
   length=20.5
   radius=8.5
   shapeName=upperarm_twist_01_l_capsule
   center=<4.0364 -1.0091 3>
   rotation=<rot 90 -90.9147 -98.9448>
  }]
 upperarm_twist_01_r [{shapeType=Capsule, length=19.5, radius=5.5, shapeName=upperarm_twist_01_r_capsule, center=<-3.7122 0 0>, rotation=<rot 90 0 0>}]
}
constraints [
 {bone1Name=dyn_hair_middle_04, bone2Name=dyn_hair_middle_03, swing1LimitDegrees=10, swing2LimitDegrees=40, twistLimitDegrees=1}
 {bone1Name=dyn_hair_middle_03, bone2Name=dyn_hair_middle_02, swing1LimitDegrees=10, swing2LimitDegrees=40, twistLimitDegrees=1}
 {bone1Name=dyn_hair_middle_02, bone2Name=dyn_hair_middle_01, swing1LimitDegrees=5, swing2LimitDegrees=40, twistLimitDegrees=30}
 {bone1Name=dyn_hair_middle_01, bone2Name=dyn_hair_middle_MSTR, swing1LimitDegrees=5, swing2LimitDegrees=40, twistLimitDegrees=1}
 {bone1Name=dyn_hair_left_04, bone2Name=dyn_hair_left_03, swing1LimitDegrees=10, swing2LimitDegrees=40, twistLimitDegrees=1}
 {bone1Name=dyn_hair_left_03, bone2Name=dyn_hair_left_02, swing1LimitDegrees=3, swing2LimitDegrees=30, twistLimitDegrees=5}
 {bone1Name=dyn_hair_left_02, bone2Name=dyn_hair_left_01, swing1LimitDegrees=3, swing2LimitDegrees=40, twistLimitDegrees=1}
 {bone1Name=dyn_hair_left_01, bone2Name=dyn_hair_left_MSTR, swing1LimitDegrees=1, swing2LimitDegrees=40, twistLimitDegrees=1}
 {bone1Name=dyn_hair_right_04, bone2Name=dyn_hair_right_03, swing1LimitDegrees=10, swing2LimitDegrees=40, twistLimitDegrees=1}
 {bone1Name=dyn_hair_right_03, bone2Name=dyn_hair_right_02, swing1LimitDegrees=3, swing2LimitDegrees=30, twistLimitDegrees=5}
 {bone1Name=dyn_hair_right_02, bone2Name=dyn_hair_right_01, swing1LimitDegrees=5, swing2LimitDegrees=40, twistLimitDegrees=1}
 {bone1Name=dyn_hair_right_01, bone2Name=dyn_hair_right_MSTR, swing1LimitDegrees=5, swing2LimitDegrees=40, twistLimitDegrees=1}
 {
  bone1Name=dyn_hair_left_03
  bone2Name=dyn_hair_middle_03
  swing1LimitDegrees=45
  swing1Motion=Free
  swing2LimitDegrees=45
  swing2Motion=Free
  twistLimitDegrees=45
  twistMotion=Free
 }
 {
  bone1Name=dyn_hair_right_03
  bone2Name=dyn_hair_middle_03
  swing1LimitDegrees=45
  swing1Motion=Free
  swing2LimitDegrees=45
  swing2Motion=Free
  twistLimitDegrees=45
  twistMotion=Free
 }
]
