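# Post-process-style Anim Blueprint for the RankedOlympus helmet face accessory, driving three dynamic hair chains from the attached parent's copied pose.
# Without back bling, hair masters follow head and neck, then run spine_05-space RigidBody physics with overridden gravity.
# With back bling, they follow spine_05, tilt, and skip simulation.
# Jumping or falling without back bling or skydiving blends in additive post-simulation bone rotations.
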AnimBlueprint /BRCosmetics/Characters/Player/Female/Medium/Bodies/F_MED_RankedOlympus_Helmet/Meshes/Parts/F_MED_RankedOlympus_Helmet_FaceAcc_AnimBP
target_skeleton=&/F_MED_RankedOlympus_Helmet_FaceAcc_Skeleton
graphs [
 AnimGraph {nodes [
   Output Pose(Result=Use cached pose 'PostSim_Pose'()) {layerGroup=None}
   n1 = "Transform (Modify) Bone - Bone: dyn_hair_left_02"!(ComponentPose=Local To Component(LocalPose=Use cached pose 'Sim_Pose'()), Alpha=1, Translation="0,0,0", Rotation="5,0,0") -> (Pose) {
    rotationMode=BMM_Additive
    rotationSpace=BCS_BoneSpace
    scaleMode=BMM_Additive
    translationMode=BMM_Additive
    translationSpace=BCS_BoneSpace
    alphaBoolBlend {blendInTime=1, blendOption=Linear, blendOutTime=1}
    boneToModify {boneName=dyn_hair_left_02}
    rotation {pitch=5}
   }
   n2 = "Transform (Modify) Bone - Bone: dyn_hair_middle_02"!(ComponentPose=@n1, Alpha=1, Translation="0,0,0", Rotation="10,0,0") -> (Pose) {
    rotationMode=BMM_Additive
    rotationSpace=BCS_BoneSpace
    scaleMode=BMM_Additive
    translationMode=BMM_Additive
    translationSpace=BCS_BoneSpace
    alphaBoolBlend {blendInTime=1, blendOption=Linear, blendOutTime=1}
    boneToModify {boneName=dyn_hair_middle_02}
    rotation {pitch=10}
   }
   n3 = "Transform (Modify) Bone - Bone: dyn_hair_right_02"!(ComponentPose=@n2, Alpha=1, Translation="0,0,0", Rotation="0,0,0") -> (Pose) {
    rotationMode=BMM_Additive
    rotationSpace=BCS_BoneSpace
    scaleMode=BMM_Additive
    translationMode=BMM_Additive
    translationSpace=BCS_BoneSpace
    alphaBoolBlend {blendInTime=1, blendOption=Linear, blendOutTime=1}
    boneToModify {boneName=dyn_hair_right_02}
   }
   n4 = "Transform (Modify) Bone - Bone: dyn_hair_left_03"!(ComponentPose=@n3, Alpha=1, Translation="0,0,0", Rotation="40,0,0") -> (Pose) {
    rotationMode=BMM_Additive
    rotationSpace=BCS_BoneSpace
    scaleMode=BMM_Additive
    translationMode=BMM_Additive
    translationSpace=BCS_BoneSpace
    alphaBoolBlend {blendInTime=1, blendOption=Linear, blendOutTime=1}
    boneToModify {boneName=dyn_hair_left_03}
    rotation {pitch=40}
   }
   n5 = "Transform (Modify) Bone - Bone: dyn_hair_middle_03"!(ComponentPose=@n4, Alpha=1, Translation="0,0,0", Rotation="40,0,0") -> (Pose) {
    rotationMode=BMM_Additive
    rotationSpace=BCS_BoneSpace
    scaleMode=BMM_Additive
    translationMode=BMM_Additive
    translationSpace=BCS_BoneSpace
    alphaBoolBlend {blendInTime=1, blendOption=Linear, blendOutTime=1}
    boneToModify {boneName=dyn_hair_middle_03}
    rotation {pitch=40}
   }
   n6 = "Transform (Modify) Bone - Bone: dyn_hair_right_03"!(ComponentPose=@n5, Alpha=1, Translation="0,0,0", Rotation="40,0,0") -> (Pose) {
    rotationMode=BMM_Additive
    rotationSpace=BCS_BoneSpace
    scaleMode=BMM_Additive
    translationMode=BMM_Additive
    translationSpace=BCS_BoneSpace
    alphaBoolBlend {blendInTime=1, blendOption=Linear, blendOutTime=1}
    boneToModify {boneName=dyn_hair_right_03}
    rotation {pitch=40}
   }
   n7 = "Transform (Modify) Bone - Bone: dyn_hair_left_04"!(ComponentPose=@n6, Alpha=1, Translation="0,0,0", Rotation="20,0,0") -> (Pose) {
    rotationMode=BMM_Additive
    rotationSpace=BCS_BoneSpace
    scaleMode=BMM_Additive
    translationMode=BMM_Additive
    translationSpace=BCS_BoneSpace
    alphaBoolBlend {blendInTime=1, blendOption=Linear, blendOutTime=1}
    boneToModify {boneName=dyn_hair_left_04}
    rotation {pitch=20}
   }
   n8 = "Transform (Modify) Bone - Bone: dyn_hair_middle_04"!(ComponentPose=@n7, Alpha=1, Translation="0,0,0", Rotation="20,0,0") -> (Pose) {
    rotationMode=BMM_Additive
    rotationSpace=BCS_BoneSpace
    scaleMode=BMM_Additive
    translationMode=BMM_Additive
    translationSpace=BCS_BoneSpace
    alphaBoolBlend {blendInTime=1, blendOption=Linear, blendOutTime=1}
    boneToModify {boneName=dyn_hair_middle_04}
    rotation {pitch=20}
   }
   n9 = "Transform (Modify) Bone - Bone: dyn_hair_right_04"!(ComponentPose=@n8, Alpha=1, Translation="0,0,0", Rotation="20,0,0") -> (Pose) {
    rotationMode=BMM_Additive
    rotationSpace=BCS_BoneSpace
    scaleMode=BMM_Additive
    translationMode=BMM_Additive
    translationSpace=BCS_BoneSpace
    alphaBoolBlend {blendInTime=1, blendOption=Linear, blendOutTime=1}
    boneToModify {boneName=dyn_hair_right_04}
    rotation {pitch=20}
   }
   Save cached pose 'PostSim_Pose'(Pose=Blend Poses by bool(
     BlendPose_0=Component To Local(ComponentPose=@n9)
     BlendPose_1=Use cached pose 'Sim_Pose'()
     BlendTime_0=1
     BlendTime_1=0.2
     bActiveValue=Get JumpFall() -> (JumpFall:bool)
    ) {blendType=Linear}) {cacheName=PostSim_Pose}
   n10 = "Transform (Modify) Bone - Bone: upperarm_l"!(ComponentPose=Local To Component(LocalPose=Use cached pose 'Start_Pose'()), Alpha=1, Rotation="0,0,0") -> (Pose) {
    rotationMode=BMM_Replace
    rotationSpace=BCS_WorldSpace
    scaleMode=BMM_Additive
    translationSpace=BCS_WorldSpace
    alphaBoolBlend {blendInTime=1, blendOption=Linear, blendOutTime=1}
    boneToModify {boneName=upperarm_l}
   }
   n11 = "Transform (Modify) Bone - Bone: upperarm_r"!(ComponentPose=@n10, Alpha=1, Rotation="0,0,0") -> (Pose) {
    rotationMode=BMM_Replace
    rotationSpace=BCS_WorldSpace
    scaleMode=BMM_Additive
    translationSpace=BCS_WorldSpace
    alphaBoolBlend {blendInTime=1, blendOption=Linear, blendOutTime=1}
    boneToModify {boneName=upperarm_r}
   }
   n12 = "Transform (Modify) Bone - Bone: upperarm_twist_01_l"!(ComponentPose=@n11, Alpha=1, Rotation="0,0,0") -> (Pose) {
    rotationMode=BMM_Replace
    rotationSpace=BCS_WorldSpace
    scaleMode=BMM_Additive
    translationSpace=BCS_WorldSpace
    alphaBoolBlend {blendInTime=1, blendOption=Linear, blendOutTime=1}
    boneToModify {boneName=upperarm_twist_01_l}
   }
   n13 = "Transform (Modify) Bone - Bone: upperarm_twist_01_r"!(ComponentPose=@n12, Alpha=1, Rotation="0,0,0") -> (Pose) {
    rotationMode=BMM_Replace
    rotationSpace=BCS_WorldSpace
    scaleMode=BMM_Additive
    translationSpace=BCS_WorldSpace
    alphaBoolBlend {blendInTime=1, blendOption=Linear, blendOutTime=1}
    boneToModify {boneName=upperarm_twist_01_r}
   }
   Save cached pose 'PreSim_Pose'(Pose=Component To Local(ComponentPose=@n13)) {cacheName=PreSim_Pose}
   n14 = "Constraint - dyn_hair_left_MSTR (P)"(ComponentPose=Local To Component(LocalPose=Use cached pose 'Input_Pose'()), Alpha=1, ConstraintWeights_0=1) {
    alphaBoolBlend {blendOption=Linear}
    boneToModify {boneName=dyn_hair_left_MSTR}
    constraintSetup [{targetBone {boneName=spine_05}}]
    constraintWeights [1]
   }
   n15 = "Constraint - dyn_hair_middle_MSTR (P)"(ComponentPose=@n14, Alpha=1, ConstraintWeights_0=1) {
    alphaBoolBlend {blendOption=Linear}
    boneToModify {boneName=dyn_hair_middle_MSTR}
    constraintSetup [{targetBone {boneName=spine_05}}]
    constraintWeights [1]
   }
   n16 = "Constraint - dyn_hair_right_MSTR (P)"(ComponentPose=@n15, Alpha=1, ConstraintWeights_0=1) {
    alphaBoolBlend {blendOption=Linear}
    boneToModify {boneName=dyn_hair_right_MSTR}
    constraintSetup [{targetBone {boneName=spine_05}}]
    constraintWeights [1]
   }
   n17 = "Transform (Modify) Bone - Bone: dyn_hair_left_02"(ComponentPose=@n16, Alpha=1, Translation="0,0,0", Rotation="-12,0,0") {
    rotationMode=BMM_Additive
    rotationSpace=BCS_BoneSpace
    scaleMode=BMM_Additive
    translationMode=BMM_Additive
    translationSpace=BCS_BoneSpace
    alphaBoolBlend {blendInTime=1, blendOption=Linear, blendOutTime=1}
    boneToModify {boneName=dyn_hair_left_02}
    rotation {pitch=-12}
   }
   n18 = "Transform (Modify) Bone - Bone: dyn_hair_middle_02"(ComponentPose=@n17, Alpha=1, Translation="0,0,0", Rotation="-12,0,0") {
    rotationMode=BMM_Additive
    rotationSpace=BCS_BoneSpace
    scaleMode=BMM_Additive
    translationMode=BMM_Additive
    translationSpace=BCS_BoneSpace
    alphaBoolBlend {blendInTime=1, blendOption=Linear, blendOutTime=1}
    boneToModify {boneName=dyn_hair_middle_02}
    rotation {pitch=-12}
   }
   n19 = "Transform (Modify) Bone - Bone: dyn_hair_right_02"(ComponentPose=@n18, Alpha=1, Translation="0,0,0", Rotation="-12,0,0") {
    rotationMode=BMM_Additive
    rotationSpace=BCS_BoneSpace
    scaleMode=BMM_Additive
    translationMode=BMM_Additive
    translationSpace=BCS_BoneSpace
    alphaBoolBlend {blendInTime=1, blendOption=Linear, blendOutTime=1}
    boneToModify {boneName=dyn_hair_right_02}
    rotation {pitch=-12}
   }
   Save cached pose 'WithBack_Pose'(Pose=Component To Local(ComponentPose=@n19)) {cacheName=WithBack_Pose}
   n20 = "Constraint - dyn_hair_left_MSTR (PP)"!(ComponentPose=Local To Component(LocalPose=Use cached pose 'Input_Pose'()), Alpha=1, ConstraintWeights_0=0.2, ConstraintWeights_1:float=0.8) -> (Pose) {
    alphaBoolBlend {blendOption=Linear}
    boneToModify {boneName=dyn_hair_left_MSTR}
    constraintSetup [{targetBone {boneName=head}}, {targetBone {boneName=neck_01}}]
    constraintWeights [1,1]
   }
   n21 = "Constraint - dyn_hair_middle_MSTR (PP)"!(ComponentPose=@n20, Alpha=1, ConstraintWeights_0=0.2, ConstraintWeights_1:float=0.8) -> (Pose) {
    alphaBoolBlend {blendOption=Linear}
    boneToModify {boneName=dyn_hair_middle_MSTR}
    constraintSetup [{targetBone {boneName=head}}, {targetBone {boneName=neck_01}}]
    constraintWeights [1,1]
   }
   n22 = "Constraint - dyn_hair_right_MSTR (PP)"!(ComponentPose=@n21, Alpha=1, ConstraintWeights_0=0.2, ConstraintWeights_1:float=0.8) -> (Pose) {
    alphaBoolBlend {blendOption=Linear}
    boneToModify {boneName=dyn_hair_right_MSTR}
    constraintSetup [{targetBone {boneName=head}}, {targetBone {boneName=neck_01}}]
    constraintWeights [1,1]
   }
   Save cached pose 'Start_Pose'(Pose=Blend Poses by bool(
     BlendPose_0=Use cached pose 'WithBack_Pose'()
     BlendPose_1=Component To Local(ComponentPose=@n22)
     bActiveValue=Get WearingBack() -> (WearingBack:bool)
    ) {blendType=Linear}) {cacheName=Start_Pose}
   Save cached pose 'Input_Pose'(Pose=Copy Pose From Mesh() {bUseAttachedParent=true}) {cacheName=Input_Pose}
   n23 = "Transform (Modify) Bone - Bone: upperarm_twist_01_l"!(
    ComponentPose=Local To Component(LocalPose=Use cached pose 'PreSim_Pose'())
    Alpha=Get Alpha_Arm_L_Up() -> (Alpha_Arm_L_Up:double)
    Translation:Vector="7.4,0,-3.2"
    Rotation="0,0,0"
   ) -> (Pose) {
    rotationMode=BMM_Additive
    rotationSpace=BCS_BoneSpace
    scaleMode=BMM_Additive
    translationMode=BMM_Additive
    translationSpace=BCS_BoneSpace
    alphaBoolBlend {blendInTime=1, blendOption=Linear, blendOutTime=1}
    boneToModify {boneName=upperarm_twist_01_l}
    translation {x=7.4, z=-3.2}
   }
   n24 = RigidBody!(
    ComponentPose=@n23
    Alpha
    OverrideWorldGravity:Vector=Get Gravity_Override() -> (Gravity_Override:Vector)
    ComponentLinearAccScale:Vector="0.4,0.4,0.4"
    ComponentLinearVelScale:Vector="0.3,0.3,0.3"
   ) -> (Pose) {
    bClampLinearTranslationLimitToRefPose=true
    bOverrideWorldGravity=true
    overridePhysicsAsset=&/F_MED_RankedOlympus_Helmet_FaceAcc_Physics
    simulationSpace=BaseBoneSpace
    alphaBoolBlend {blendOption=Linear}
    alphaScaleBiasClamp {interpSpeedDecreasing=4, interpSpeedIncreasing=6}
    baseBoneRef {boneName=spine_05}
    componentAppliedLinearAccClamp=<3000 3000 3000>
    componentLinearAccScale=<0.4 0.4 0.4>
    componentLinearVelScale=<0.3 0.3 0.3>
    overrideWorldGravity {z=-875}
    simSpaceSettings {maxAngularAcceleration=200, maxAngularVelocity=5, maxLinearAcceleration=500, maxLinearVelocity=200}
   }
   Save cached pose 'Sim_Pose'(Pose=Blend Poses by bool(BlendPose_0=Use cached pose 'PreSim_Pose'(), BlendPose_1=Component To Local(ComponentPose=@n24), bActiveValue=Get WearingBack()) {blendType=Linear}) {cacheName=Sim_Pose}
   Use cached pose 'Sim_Pose'()
  ]}
 EventGraph {nodes [
   Event BlueprintUpdateAnimation()
   > n1 = Sequence!(execute) -> (then_0, then_1, then_2:Exec)
   n2 = Get BodyMesh() -> (BodyMesh:Skeletal Mesh Component Object Reference)
   > Is Valid(InputObject=@n2)
   > Set Upperarm_L_Rotation(Upperarm_L_Rotation:Rotator=BreakTransform(InTransform=GetSocketTransform(self=@n2, InSocketName=upperarm_l, TransformSpace=RTS_World)).Rotation) -> (Output_Get:Rotator)
   > Set Upperarm_R_Rotation(Upperarm_R_Rotation:Rotator=BreakTransform(InTransform=GetSocketTransform(self=@n2, InSocketName=upperarm_r, TransformSpace=RTS_World)).Rotation) -> (Output_Get:Rotator)
   > Set Upperarm_Twist_L_Rotation(Upperarm_Twist_L_Rotation:Rotator=BreakTransform(InTransform=GetSocketTransform(self=@n2, InSocketName=upperarm_twist_01_l, TransformSpace=RTS_World)).Rotation) -> (Output_Get:Rotator)
   > Set Upperarm_Twist_R_Rotation(Upperarm_Twist_R_Rotation:Rotator=BreakTransform(InTransform=GetSocketTransform(self=@n2, InSocketName=upperarm_twist_01_r, TransformSpace=RTS_World)).Rotation) -> (Output_Get:Rotator)
   n3 = Get SkelMesh()
   Set Alpha_Head_Bk(
    execute=@n1.then_2
    Alpha_Head_Bk:double=GetAlphaFromReferenceTransformDelta(SkeletalMeshComponent=@n3, BoneName=head, BaseName=spine_05, InputRange=(X=-25,Y=-60), TwistAxis="1,0,0")
   ) -> (Output_Get:double)
   > Set Alpha_Arm_L_Up(Alpha_Arm_L_Up:double=GetAlphaFromReferenceTransformDelta(SkeletalMeshComponent=@n3, BoneName=upperarm_l, BaseName=clavicle_l, DeltaComponentType=EulerY, InputRange=(X=20,Y=60), TwistAxis="1,0,0")) -> (Output_Get:double)
   > Set WearingBack(WearingBack:bool=CheckForBackbling()) -> (Output_Get:bool)
   n4 = Get WearingBack()
   n5 = NOT Boolean(A=Get bIsSkyDiving())
   n6 = OR Boolean(A=Get bIsFalling(), B=Get bIsJumping())
   > Set JumpFall(JumpFall:bool=AND Boolean(A=AND Boolean(A=@n5, B=@n6), B=NOT Boolean(A=@n4))) -> (Output_Get:bool)
   n7 = Gravity Override State Function DEPRECATED(
    execute=@n1.then_1
    CharacterPartAnimInstance=Self-Reference()
    GameTimeInSeconds=GetGameTimeInSeconds()
    GravityOverrideParams=Get Gravity_Override_Params() -> (Gravity_Override_Params:Gravity Override Params Struct Structure)
   )
   > Set Gravity_Override(Gravity_Override:Vector=@n7) -> (Output_Get:Vector)
   n8 = Get bIsSkyDiving()
   n9 = Get bIsFalling()
   n10 = SelectVector(A="0.3,0.3,0.3", B=SelectVector(A="0.1,0.1,0.1", B="0.4,0.4,0.4", bPickA=AND Boolean(A=@n8, B=@n9)), bPickA=Get JumpFall())
   > Set Component_Linear_AccScale(Component_Linear_AccScale:Vector=@n10) -> (Output_Get:Vector)
   > Set Component_Linear_VelScale(Component_Linear_VelScale:Vector=@n10) -> (Output_Get:Vector)
   n11 = Event OnCharacterCustomizationLoaded()
   > Set BodyMesh(BodyMesh=Get Array Item with EFort Custom Part Type(Array=@n11.CustomMeshParts))
   Event BlueprintInitializeAnimation()
   > Set SkelMesh(SkelMesh=GetOwnerSkeletalMeshComponent())
  ]}
]
variables [
 Alpha_Arm_L_Up {type=number}
 Alpha_Head_Bk {type=number}
 BodyMesh {type=SkeletalMeshComponent}
 Component_Linear_AccScale {type=Vector}
 Component_Linear_VelScale {type=Vector}
 Gravity_Override {type=Vector}
 Gravity_Override_Params {type=GravityOverrideParamsStruct}
 JumpFall {type=boolean}
 SkelMesh {type=SkeletalMeshComponent}
 Upperarm_L_Rotation {type=Rotator}
 Upperarm_R_Rotation {type=Rotator}
 Upperarm_Twist_L_Rotation {type=Rotator}
 Upperarm_Twist_R_Rotation {type=Rotator}
 WearingBack {type=boolean}
]
